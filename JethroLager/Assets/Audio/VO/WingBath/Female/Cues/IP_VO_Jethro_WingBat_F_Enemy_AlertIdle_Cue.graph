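# Sound cue for female WingBat enemy alert-idle voice lines in the JethroLager content.
# A Random node picks among six equally weighted wave variations.
# It plays as non-critical enemy VO with shared NPC alerted attenuation and henchmen combat chatter concurrency.
# Bus sends feed music, weapon, first-person weapon and explosion sidechain volume buses.
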SoundCue /JethroLager/Assets/Audio/VO/WingBath/Female/Cues/IP_VO_Jethro_WingBat_F_Enemy_AlertIdle_Cue
root=Random_0
nodes [
 {id=Random_0, children [WavePlayer_0,WavePlayer_1,WavePlayer_2,WavePlayer_3,WavePlayer_4,WavePlayer_5], properties {weights [1,1,1,1,1,1]}}
 {id=WavePlayer_0, properties {soundWaveAssetPtr=/JethroLager/Assets/Audio/VO/WingBath/Female/Waves/IP_VO_Jethro_WingBat_F_AlertIdle_011}}
 {id=WavePlayer_1, properties {soundWaveAssetPtr=/JethroLager/Assets/Audio/VO/WingBath/Female/Waves/IP_VO_Jethro_WingBat_F_AlertIdle_012}}
 {id=WavePlayer_2, properties {soundWaveAssetPtr=/JethroLager/Assets/Audio/VO/WingBath/Female/Waves/IP_VO_Jethro_WingBat_F_AlertIdle_022}}
 {id=WavePlayer_3, properties {soundWaveAssetPtr=/JethroLager/Assets/Audio/VO/WingBath/Female/Waves/IP_VO_Jethro_WingBat_F_AlertIdle_032}}
 {id=WavePlayer_4, properties {soundWaveAssetPtr=/JethroLager/Assets/Audio/VO/WingBath/Female/Waves/IP_VO_Jethro_WingBat_F_AlertIdle_040}}
 {id=WavePlayer_5, properties {soundWaveAssetPtr=/JethroLager/Assets/Audio/VO/WingBath/Female/Waves/IP_VO_Jethro_WingBat_F_AlertIdle_043}}
]
properties {
 attenuationSettings=/Game/Sounds/Fort_Human_VO_BR/NPCs/Shared/NPC_VO_Alerted_Att
 maxDistance=12500
 soundClassObject=/Game/Sounds/FortSoundClassesAndModes/Fort_VO_Enemy_NonCritical
 volumeMultiplier=1.0525
 busSends [
  {audioBus=&AB_Sidechain_FortMusic_Vol, minSendLevel=0, sendLevel=0.25, soundSourceBus=null}
  {audioBus=/Game/Sounds/Modulation/EnvelopeFollowers/Volume/AB_Sidechain_FortWeapon1P_Vol, minSendLevel=0, sendLevel=0.25, soundSourceBus=null}
  {audioBus=&AB_Sidechain_FortWeapon_Vol, minSendLevel=0, sendLevel=0.25, soundSourceBus=null}
  {
   audioBus=&AB_Sidechain_FortExplosion_Vol
   minSendLevel=0
   sendLevel=0.25
   soundSourceBus=null
  }
 ]
 concurrencySet [/Game/Sounds/Fort_Human_VO_BR/NPCs/Shared/NPC_Henchmen_Combat_Chatter_Conc]
}
